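# Sound cue for exiting the WilliePete secret passage hideable, filed under the Whirlpool variant.
# A random node picks between two exit wave variations with equal weight.
# A modulator then applies a slight randomized upward pitch shift.
# The cue uses trap attenuation and the gadget items sound class, and is set to prime on load.
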SoundCue /FNGameplayCues/Sounds/Fort_Hideables/SecretPassage/Whirlpool/WilliePete_Exit_Cue
root=Modulator_0
nodes [
 {id=Modulator_0, children [Random_0], properties {pitchMax=1.15, pitchMin=1.1}}
 {id=Random_0, children [WavePlayer_2,WavePlayer_3], properties {weights [1,1]}}
 {id=WavePlayer_2, properties {soundWaveAssetPtr=&/WilliePete_Exit_01}}
 {id=WavePlayer_3, properties {soundWaveAssetPtr=&/WilliePete_Exit_02}}
]
properties {
 attenuationSettings=&Fort_Traps_Att
 bPrimeOnLoad=true
 maxDistance=6000
 soundClassObject=&Fort_Gadgets_Items
 volumeMultiplier=1
}
